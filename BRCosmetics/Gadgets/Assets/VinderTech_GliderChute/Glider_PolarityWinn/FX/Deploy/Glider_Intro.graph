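# Level Sequence that plays the PolarityWinn glider's deploy animation alongside its Niagara deploy effect.
# It possesses a FortSkeletalMeshActor held at a fixed transform, with one skeletal animation section and a second, empty animation track.
# The NS_PolarityWinnGlider_Deploy Niagara component is bound as a child of that actor and driven by a system life-cycle track that runs longer than the animation section.
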LevelSequence /BRCosmetics/Gadgets/Assets/VinderTech_GliderChute/Glider_PolarityWinn/FX/Deploy/Glider_Intro
bindings [
 PolarityWinn_Glider {
  id=#1
  possessed_class=FortSkeletalMeshActor
  tracks [
   {
    type=MovieScene3DTransformTrack
    sections [{
      type=MovieScene3DTransformSection
      channels [
       Location.X {type=MovieSceneScriptingDoubleChannel, default=114.5207}
       Location.Y {type=MovieSceneScriptingDoubleChannel, default=1452.7815}
       Location.Z {type=MovieSceneScriptingDoubleChannel, default=2047.9451}
       Rotation.X {type=MovieSceneScriptingDoubleChannel, default=0}
       Rotation.Y {type=MovieSceneScriptingDoubleChannel, default=0}
       Rotation.Z {type=MovieSceneScriptingDoubleChannel, default=0}
       Scale.X {type=MovieSceneScriptingDoubleChannel, default=1}
       Scale.Y {type=MovieSceneScriptingDoubleChannel, default=1}
       Scale.Z {type=MovieSceneScriptingDoubleChannel, default=1}
      ]
     }]
   }
   {type=MovieSceneSkeletalAnimationTrack, sections [{type=MovieSceneSkeletalAnimationSection, end_frame=160, start_frame=0}]}
   {type=MovieSceneSkeletalAnimationTrack, sections []}
  ]
 }
 NS_PolarityWinnGlider_Deploy {
  id=#2
  parent_id=#1
  possessed_class=NiagaraComponent
  tracks [{type=MovieSceneNiagaraSystemTrack, sections [{type=MovieSceneNiagaraSystemSpawnSection, end_frame=300, start_frame=0}]}]
 }
]
display_rate {denominator=1, numerator=30}
tick_resolution {denominator=1, numerator=24000}
tracks []
