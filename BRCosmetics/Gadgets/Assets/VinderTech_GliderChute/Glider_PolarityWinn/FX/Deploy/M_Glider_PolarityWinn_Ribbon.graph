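# Translucent unlit Niagara ribbon material for the Polarity Winn glider's deploy effect, drawing an animated energy trail.
# Emissive combines curve-atlas color, panning smoke noise and RGB-offset spline pulses, with dynamic parameters driving speed energy, pulse intensity and a noisy appear flash.
# Opacity combines particle alpha, edge fade, noise erosion and an Appear-driven wipe along U.
# A clamped offset from particle position toward a dynamic-parameter value, weighted along U by WPOIntensity, bends the ribbon.
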Material /BRCosmetics/Gadgets/Assets/VinderTech_GliderChute/Glider_PolarityWinn/FX/Deploy/M_Glider_PolarityWinn_Ribbon
graph {
 properties {
  bUsedWithNiagaraMeshParticles=true
  bUsedWithNiagaraRibbons=true
  blendMode=BLEND_Translucent
  pixelDepthOffsetMode=PDOM_Legacy
  shadingModel=MSM_Unlit
  usedShadingModels=Unlit
 }
 nodes [
  AppendVector(
   A=Constant()
   B=LinearInterpolate(Alpha=MaterialFunctionCall() {materialFunction=&DebugTimeSine}.Result) {constB=0.2}
  )
  Constant2Vector() {g=1}
  n1 = DynamicParameter() {parameterIndex=1, paramNames [X,Y,Z,WPOIntensity]}
  n2 = Subtract(
   A=MaterialFunctionCall(R=@n1.X, G=@n1.X, B=@n1.X) {materialFunction=&AppendMany}.RGB
   B=ParticlePositionWS()
  )
  n3 = ComponentMask(None=@n2) {r=true}
  n4 = Constant() {r=-50}
  n5 = Constant() {r=50}
  n6 = ComponentMask(None=@n2) {g=true}
  n7 = Constant() {r=-50}
  n8 = Constant() {r=50}
  n9 = ComponentMask(None=@n2) {b=true}
  n10 = Constant() {r=-10}
  n11 = Constant() {r=10}
  n12 = ComponentMask(None=TextureCoordinate()) {r=true}
  n13 = ScalarParameter() {defaultValue=2, parameterName=Interpolate Blend}
  LinearInterpolate(
   B=MaterialFunctionCall(X=Clamp(None=@n3, Min=@n4, Max=@n5), Y=Clamp(None=@n6, Min=@n7, Max=@n8), Z=Clamp(None=@n9, Min=@n10, Max=@n11)) {materialFunction=&MakeFloat3}.Result
   Alpha=Multiply(A=@n1.WPOIntensity, B=Power(Base=@n12, Exp=@n13))
  )
  n14 = VectorParameter() {parameterName=NoiseColor02, defaultValue=<rgba 2 0 0.1236 1>}
  n15 = VectorParameter() {parameterName=NoiseColor01, defaultValue=<rgba 0 1.902 3 1>}
  n16 = Add(A=TextureCoordinate() {uTiling=0.6}, B=Constant2Vector() {g=0.7})
  n17 = Panner() {speedX=-0.2}
  n18 = Multiply(A=ComponentMask(None=TextureSample(UVs=@n17) {texture=&T_Inky_Smoke_Tile}.RGB) {g=true, r=true}) {constB=0.15}
  n19 = TextureSample(UVs=Panner(Coordinate=Add(A=@n16, B=@n18), Speed=Constant2Vector() {r=-1}) {speedX=-0.6}) {texture=&T_Inky_Smoke_Tile}
  n20 = Power(Base=@n19.R)
  n21 = Desaturation(None=Multiply(A=LinearInterpolate(A=@n14.RGB, B=@n15.RGB, Alpha=@n19.R), B=Multiply(A=@n20) {constB=0.7}), Fraction=Constant() {r=-1.5})
  n22 = Multiply(A=Multiply(A=@n21) {constB=0.6}, B=ScalarParameter() {defaultValue=0.5, parameterName=Idle noise})
  n23 = Add(A=TextureCoordinate() {uTiling=-1.3}, B=Constant2Vector() {g=0.2})
  n24 = TextureSample(UVs=Panner(Coordinate=TextureCoordinate()) {speedY=0.5}) {samplerType=SAMPLERTYPE_LinearGrayscale, texture=/Engine/Functions/Engine_MaterialFunctions02/ExampleContent/Textures/LowResBlurredNoise}
  n25 = ComponentMask(None=@n24.RGB) {g=true, r=true}
  n26 = Constant() {r=0.3}
  n27 = Multiply(A=@n25, B=@n26) {constB=0.3}
  n28 = Add(A=Subtract(A=@n23), B=@n27)
  n29 = Constant2Vector() {g=1.8}
  n30 = TextureSample(UVs=Panner(Coordinate=MaterialFunctionCall(XY=@n28) {materialFunction=&Swizzle}.YX, Speed=@n29) {speedY=1}) {samplerSource=SSM_Wrap_WorldGroupSettings, texture=/Game/Accessories/FORT_Backpacks/Backpack_M_MED_OceanBreeze/FX/T_OceanBreeze_SplineB_t}
  n31 = ScalarParameter() {defaultValue=0.02, parameterName=ColorOffset}
  n32 = Add(A=Multiply(A=Add(A=@n23, B=@n31)) {constB=1.1}, B=Multiply(A=@n25, B=@n26) {constB=0.3})
  n33 = TextureSample(UVs=Panner(Coordinate=MaterialFunctionCall(XY=@n32) {materialFunction=&Swizzle}.YX, Speed=@n29) {speedY=1}) {samplerSource=SSM_Wrap_WorldGroupSettings, texture=&T_OceanBreeze_SplineB_t}
  n34 = Add(A=Subtract(A=@n23, B=@n31), B=Multiply(A=@n25, B=@n26) {constB=0.3})
  n35 = TextureSample(UVs=Panner(Coordinate=MaterialFunctionCall(XY=@n34) {materialFunction=&Swizzle}.YX, Speed=@n29) {speedY=1}) {samplerSource=SSM_Wrap_WorldGroupSettings, texture=&T_OceanBreeze_SplineB_t}
  n36 = Multiply(
   A=Constant3Vector() {constant=<rgba 1.1005 1.7992 2 1>}
   B=Power(Base=MaterialFunctionCall(R=@n33.R, G=@n30.R, B=@n35.R) {materialFunction=&AppendMany}.RGB) {constExponent=0.5}
  )
  n37 = LinearInterpolate(
   A=VectorParameter() {parameterName=PulseColor02, defaultValue=<rgba 1 0 0.6646 1>}.RGB
   B=VectorParameter() {parameterName=PulseColor01, defaultValue=<rgba 0.1429 0.0781 1 1>}.RGB
   Alpha=@n36
  )
  n38 = Multiply(
   A=Multiply(A=Multiply(A=@n37, B=@n36)) {constB=2.5}
   B=DynamicParameter() {paramNames [Speed energy, Appear, Pulse instensity, Param4]}.Pulse instensity
  )
  n39 = ComponentMask(None=TextureCoordinate()) {g=true}
  n40 = CurveAtlasRowParameter(CurveTime=LinearInterpolate(A=@n39, B=OneMinus(None=@n39), Alpha=ScalarParameter() {parameterName=Use invert mask})) {
   atlas=/BRCosmetics/Gadgets/Assets/VinderTech_GliderChute/Glider_InstantGravel/FX/CAtlas_Glider_InstantGravel_Color
   curve=/BRCosmetics/Gadgets/Assets/VinderTech_GliderChute/Glider_InstantGravel/FX/CV_Glider_InstantGravel_Color
   parameterName=Color
  }
  n41 = ParticleColor()
  n42 = Add(
   A=LinearInterpolate(B=Add(A=@n21, B=@n38) {constB=0}, Alpha=DynamicParameter() {paramNames [Speed energy, Appear, Pulse instensity, Param4]}.Speed energy)
   B=Multiply(A=@n40, B=@n41.RGB)
  )
  n43 = LinearInterpolate(Alpha=DynamicParameter() {paramNames [Speed energy, Appear, Pulse instensity, Param4]}.Appear) {constA=-0.1, constB=1.8}
  n44 = Add(A=Add(A=TextureCoordinate() {uTiling=-2}, B=Constant2Vector() {g=0.2}) {constB=0}, B=AppendVector(A=Add(A=@n43) {constB=0.2}, B=Constant()))
  n45 = TextureSample(UVs=MaterialFunctionCall(XY=Add(A=@n44, B=@n27)) {materialFunction=&Swizzle}.YX) {samplerSource=SSM_Clamp_WorldGroupSettings, texture=&T_OceanBreeze_SplineB_t}
  n46 = Power(Base=@n45.R) {constExponent=1}
  n47 = TextureSample(UVs=TextureCoordinate() {vTiling=0.5}) {samplerType=SAMPLERTYPE_Normal, texture=/Game/Characters/Player/Female/Medium/Bodies/F_MED_Bistro_Astronaut/FX/T_GaussianNoise_Normal}
  n48 = Multiply(A=@n47.RGB, B=ScalarParameter() {defaultValue=250, parameterName=Appear Noise}) {constB=100}
  MaterialFunctionCall(Emissive=Add(A=Add(A=@n22, B=@n42), B=Multiply(A=@n46, B=@n48))) {materialFunction=&MF_FX_EmissiveAdjustments}
  n49 = TextureCoordinate() {uTiling=0}
  n50 = Constant2Vector() {g=0.5}
  n51 = ScalarParameter() {defaultValue=0.5, parameterName=Edge fade rang}
  n52 = ScalarParameter() {defaultValue=0.5, parameterName=Edge fade hardness}
  n53 = OneMinus(None=ComponentMask(None=TextureCoordinate()) {r=true})
  n54 = TextureSample(UVs=Panner(Coordinate=TextureCoordinate() {uTiling=0.5}) {speedX=-1}) {samplerType=SAMPLERTYPE_LinearGrayscale, texture=&LowResBlurredNoise}
  n55 = ComponentMask(None=TextureCoordinate()) {r=true}
  n56 = Multiply(A=Saturate(None=Multiply(A=@n53) {constB=2}), B=Power(Base=@n54.R, Exp=Power(Base=@n55)))
  n57 = ComponentMask(None=TextureCoordinate()) {r=true}
  n58 = Multiply(
   A=Multiply(A=SphereMask(A=@n49, B=@n50, Radius=@n51, Hardness=@n52) {attenuationRadius=0.5, hardnessPercent=50}, B=Power(Base=@n56))
   B=Saturate(None=Multiply(A=@n57) {constB=10})
  )
  n59 = TextureCoordinate()
  n60 = Add(
   A=OneMinus(None=ComponentMask(None=@n59) {r=true})
   B=LinearInterpolate(Alpha=DynamicParameter() {paramNames [Speed energy, Appear, Pulse instensity, Param4]}.Appear) {constA=-1, constB=0.1}
  ) {constB=-1}
  Multiply(A=Multiply(A=Multiply(A=@n41.A, B=@n58)) {constB=0.8}, B=Saturate(None=Multiply(A=@n60) {constB=10}))
 ]
}
