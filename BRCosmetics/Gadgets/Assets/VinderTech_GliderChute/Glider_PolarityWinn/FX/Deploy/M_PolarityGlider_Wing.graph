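# Translucent unlit, two-sided Niagara material for the wing deploy effect on the Polarity glider, drawing animatable glowing lines.
# A sine of flowmap green plus a dynamic-parameter offset forms the lines, with count and sharpness also particle-driven.
# A blurred flowmap raised to a power masks them, along with particle alpha and a depth fade skipped on Low and Medium quality.
# A three-color UV gradient tinted by particle color feeds emissive adjustments.
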Material /BRCosmetics/Gadgets/Assets/VinderTech_GliderChute/Glider_PolarityWinn/FX/Deploy/M_PolarityGlider_Wing
graph {
 properties {
  bAutomaticallySetUsageInEditor=false
  bUsedWithNiagaraMeshParticles=true
  bUsedWithNiagaraSprites=true
  blendMode=BLEND_Translucent
  shadingModel=MSM_Unlit
  twoSided=true
  usedShadingModels=Unlit
 }
 nodes [
  n1 = VectorParameter() {parameterName=LinesColor1, defaultValue=<rgba 0.2442 1.0909 2 1>}
  n2 = VectorParameter() {parameterName=LinesColor2, defaultValue=<rgba 0.8268 0.2101 1 1>}
  n3 = VectorParameter() {parameterName=LinesColor3, defaultValue=<rgba 0.1462 0.0596 0.3281 1>}
  n4 = ComponentMask(None=TextureCoordinate()) {r=true}
  n5 = Add(A=@n4, B=ScalarParameter() {parameterName=LinesColorGradientBias}) {constB=0}
  n6 = Power(Base=OneMinus(None=Saturate(None=@n5)), Exp=ScalarParameter() {defaultValue=1, parameterName=Power}) {constExponent=1}
  n7 = ParticleColor()
  MaterialFunctionCall(Emissive=Multiply(
    A=MaterialFunctionCall(A=@n1.RGB, B=@n2.RGB, C=@n3.RGB, Alpha=@n6) {materialFunction=&Lerp_3Color}.Result
    B=@n7.RGB
   )) {materialFunction=&MF_FX_EmissiveAdjustments}
  n8 = TextureCoordinate()
  n9 = TextureSample(UVs=OneMinus(None=MaterialFunctionCall(XY=@n8) {materialFunction=&Swizzle}.YX)) {
   samplerSource=SSM_Clamp_WorldGroupSettings
   samplerType=SAMPLERTYPE_LinearGrayscale
   texture=/BRCosmetics/Gadgets/Assets/VinderTech_GliderChute/Glider_PolarityWinn/Textures/T_Polarity_Glider_flowmap
  }
  n10 = DynamicParameter() {defaultValue=<rgba 0 1 1 1>, paramNames [Offset Lines, LineExponent, Line Number, Mask2]}
  n11 = Add(A=@n9.G, B=@n10.Offset Lines)
  n12 = ConstantBiasScale(None=Sine(None=Multiply(A=@n11, B=@n10.Line Number) {constB=4}))
  n13 = Add(A=@n10.LineExponent, B=ScalarParameter() {parameterName=LineExponent})
  n14 = NamedRerouteDeclaration(None=Saturate(None=Power(Base=@n12, Exp=@n13) {constExponent=6})) {name=SineWaves, nodeColor=<rgba 0.8281 0 1 1>}
  n15 = TextureCoordinate()
  n16 = TextureSample(UVs=OneMinus(None=MaterialFunctionCall(XY=@n15) {materialFunction=&Swizzle}.YX)) {
   samplerSource=SSM_Clamp_WorldGroupSettings
   samplerType=SAMPLERTYPE_LinearGrayscale
   texture=&/T_Polarity_Glider_flowmap_Blur_PNG
  }
  n17 = Power(Base=@n16.G, Exp=NamedRerouteUsage())
  n18 = DepthFade(FadeDistance=ScalarParameter() {defaultValue=60, parameterName=DepthFade})
  n19 = Constant() {r=1}
  Multiply(A=Multiply(A=@n7.A, B=Multiply(A=@n14, B=@n17)), B=QualitySwitch(Default=@n18, Low=@n19, High=@n18, Medium=@n19, Epic=@n18))
  NamedRerouteDeclaration(None=@n10.Mask2) {name=Mask, nodeColor=<rgba 1 0 0.6328 1>}
  OneMinus(None=@n4)
  ScalarParameter() {defaultValue=3, parameterName=LineNumber}
 ]
}
