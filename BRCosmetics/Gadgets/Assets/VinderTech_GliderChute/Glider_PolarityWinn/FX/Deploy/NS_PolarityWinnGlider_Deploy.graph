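# Niagara deploy effect for the PolarityWinn glider, tinted between two palettes by User.Variant and scaled by a time-of-day color multiplier.
# Deploy_Trail draws curl-noise anim-trail ribbons, while mirrored wing emitters burst socket-attached sprites and meshes after a fixed delay.
# A socket-following flare with light and expanding rings fire after User.DeployDelay.
# Front-end modifiers adjust timing, offsets and sizes, and choose between two near-identical ring emitters via spawn probability.
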NiagaraSystem /BRCosmetics/Gadgets/Assets/VinderTech_GliderChute/Glider_PolarityWinn/FX/Deploy/NS_PolarityWinnGlider_Deploy
dependencies {
 usedDataInterfaces [
  &NiagaraDataInterfaceCurve
  &NiagaraDataInterfaceVector2DCurve
  &NiagaraDataInterfaceColorCurve
  &NiagaraDataInterfaceVectorCurve
  &NiagaraDataInterfaceSocketReader
 ]
 usedDynamicInputs [
  &AddVectorToPosition
  &Multiply_VectorByFloat
  &SetBoolByFloatComparison
 ]
 usedModules [
  &self:SystemSpawnScript.SystemScriptSource.SystemScriptGraph.NiagaraNodeAssignment_3.SetVariables_#1
  &self:SystemSpawnScript.SystemScriptSource.SystemScriptGraph.NiagaraNodeAssignment_0.SetVariables_#2
  &self:SystemSpawnScript.SystemScriptSource.SystemScriptGraph.NiagaraNodeAssignment_2.SetVariables_#3
  &self:SystemSpawnScript.SystemScriptSource.SystemScriptGraph.NiagaraNodeAssignment_5.SetVariables_#4
  &self:SystemSpawnScript.SystemScriptSource.SystemScriptGraph.NiagaraNodeAssignment_1.SetVariables_#5
  &SystemState
  &self:SystemSpawnScript.SystemScriptSource.SystemScriptGraph.NiagaraNodeAssignment_4.SetVariables_#6
 ]
}
emitters {
 Deploy_Trail {
  emitterUpdateScript [
   &EmitterState(
    Inactive Response=::NewEnumerator1
    Life Cycle Mode=::NewEnumerator1
    Loop Behavior=::NewEnumerator1
    Loop Delay=2
    Loop Duration=1.3
    UseLoopDelay=false
   )
   &FN_SpawnAnimTrail() {module=FN_SpawnAnimTrail001}
  ]
  particleSpawnScript [
   &FN_InitializeAnimTrailBoneIndecies(Second Bone or Socket Index=0)
   &FN_InitializeAnimTrail(AnimTrailWidth=25, Lifetime=1)
   &AddVelocity(Rotation Coordinate Space=::Simulation, Velocity=<150 500 0>) {enabled=false}
   &InheritVelocity(Inherited Velocity Speed Limit=25, Speed Limit=true, UseSpeedThreshold=false)
  ]
  particleUpdateScript [
   &ParticleState()
   &FN_UpdateAnimTrail()
   &ScaleRibbonWidth(Ribbon Width Scale=&FloatFromCurve(FloatCurve=DataInterface() {propertyValues {Curve {
        defaultValue=1
        keys [{interpMode=RCIM_Cubic, tangentMode=RCTM_User, time=0, value=1}, {interpMode=RCIM_Cubic, tangentMode=RCTM_User, time=1, value=0.25}]
       }}}))
   &CurlNoiseForce(
    Bias Noise=false
    Noise Frequency=&FN_FrontEndFloatModifier(Gameplay=5, Front End=10)
    Noise Strength=&RandomRangeFloat(Minimum=200, Maximum=300)
    Pan Noise=false
   )
   &SolveForcesAndVelocity(Clamp Velocity=false, Limit Acceleration=false)
   &ScaleColor(
    Scale Alpha=&Multiply_Float(
     A=&Multiply_Float(
      A=&MakeFloatFromLinearColor(LinearColor=$System.TODColorMultiplier, Channel:ENiagaraFloatFromLinearColorOptions=::NewEnumerator3)
      B=&FloatFromCurve(
       FloatCurve=DataInterface() {propertyValues {Curve {keys [
           {interpMode=RCIM_Cubic, tangentMode=RCTM_User, time=0, value=1}
           {arriveTangent=-2.023, interpMode=RCIM_Cubic, leaveTangent=-2.023, tangentMode=RCTM_User, time=1, value=0}
          ]}}}
       Scale Curve=&FloatFromCurve(
        FloatCurve=DataInterface() {propertyValues {Curve {keys [<at 0.5 0 RCIM_Cubic>,<at 0.8 1 RCIM_Cubic>,<at 1 0 RCIM_Cubic>]}}}
        CurveIndex=$Emitter.Age
       )
      )
     )
     B=&Lerp_Float(A=1, B=0.7, Alpha=$User.Variant)
    )
    Scale RGB=&MakeVectorFromLinearColorRGB(Color=$System.TODColorMultiplier)
    ScaleA=true
    ScaleRGB=true
   ) {module=ScaleColor001}
   /SetVariables(Particles.Lines Color:LinearColor=&MakeLinearColorFromVectorAndFloat(
     "Vector (RGB)"=&MakeVectorFromLinearColorRGB(Color=&ColorFromCurve(ColorCurve=DataInterface() {propertyValues {
         AlphaCurve {keys [<at 0 1 RCIM_Cubic>,<at 1 0 RCIM_Cubic>]}
         BlueCurve {keys [<at 0 2.839 RCIM_Cubic>,<at 0.0677 2.6142>,<at 0.2582 1>,<at 0.6407 0.7974 RCIM_Cubic>]}
         GreenCurve {keys [<at 0 1.2451 RCIM_Cubic>,<at 0.0677 0.5172>,<at 0.2582 0.232>,<at 0.6407 0.2741 RCIM_Cubic>]}
         RedCurve {keys [<at 0 3 RCIM_Cubic>,<at 0.0677 2.745>,<at 0.2582 0.5677>,<at 0.6407 0.2198 RCIM_Cubic>]}
        }}))
     "Float (Alpha)"=&Multiply_Float(
      A=&FloatFromCurve(
       FloatCurve=DataInterface() {propertyValues {Curve {keys [
           {interpMode=RCIM_Cubic, tangentMode=RCTM_User, time=0, value=1}
           {arriveTangent=-2.023, interpMode=RCIM_Cubic, leaveTangent=-2.023, tangentMode=RCTM_User, time=1, value=0}
          ]}}}
       Scale Curve=&FloatFromCurve(
        FloatCurve=DataInterface() {propertyValues {Curve {keys [<at 0.5 0 RCIM_Cubic>,<at 0.8 1 RCIM_Cubic>,<at 1 0 RCIM_Cubic>]}}}
        CurveIndex=$Emitter.Age
       )
      )
      B=&FloatFromCurve(
       FloatCurve=DataInterface() {propertyValues {Curve {keys [
           {arriveTangent=-0.199, interpMode=RCIM_Cubic, leaveTangent=-0.199, tangentMode=RCTM_User, time=0, value=1}
           {arriveTangent=-2.013, tangentMode=RCTM_User, time=1, value=0}
          ]}}}
       CurveIndex=&ReturnNormalizedExecIndex()
      )
     )
    )) {isSetParameters=true}
   &DynamicMaterialParameters(
    Index 0 Param 1=1
    Index 0 Param 2=&FloatFromCurve(FloatCurve=DataInterface() {propertyValues {Curve {
        defaultValue=1
        keys [{arriveTangent=0.9489, interpMode=RCIM_Cubic, leaveTangent=0.9489, tangentMode=RCTM_User, time=0, value=0.4}, <at 1 1 RCIM_Cubic>]
       }}})
    Index 0 Param 3=1
    Module.Index0Float0Write=true
    Module.Index0Float1Write=true
    Module.Index0Float2Write=true
    Module.Index0Float3Write=false
   ) {module=DynamicMaterialParameters001}
  ]
  rendererClasses [&NiagaraRibbonRendererProperties]
  renderers [
   {rendererClass=&NiagaraRibbonRendererProperties}
   {rendererClass=&NiagaraRibbonRendererProperties, rendererIndex=1}
   {rendererClass=&NiagaraRibbonRendererProperties, rendererIndex=2}
  ]
 }
 LeftWing {
  emitterSpawnScript [/SetVariables(
    Emitter.CenterColor:LinearColor=&LerpLinearColors(StartColor=<rgba 1.0887 0.3246 2 1>, EndColor=<rgba 1.7272 0.3631 2 1>, LerpFactor=$User.Variant)
    Emitter.EdgeColor:LinearColor=&LerpLinearColors(StartColor=<rgba 2 0.713 1.8132 1>, EndColor=<rgba 1 0 3 1>, LerpFactor=$User.Variant)
    Emitter.TipColor:LinearColor=&LerpLinearColors(StartColor=<rgba 0.2218 0.6687 1.5 1>, EndColor=<rgba 0.2218 1.5 1.5 1>, LerpFactor=$User.Variant)
   ) {isSetParameters=true, module=SetVariables_#7}]
  emitterUpdateScript [
   &EmitterState(
    Life Cycle Mode=::NewEnumerator1
    Loop Behavior=::NewEnumerator1
    Loop Delay=&FN_FrontEndFloatModifier(Gameplay=2.1, Front End=1.83)
    Loop Duration=1
    UseLoopDelay=true
   )
   &SpawnBurst_Instantaneous(Spawn Count=3, Use Spawn Probability=false, UseLoopCountLimit=false)
  ]
  particleSpawnScript [
   &InitializeParticle(
    Color Mode=::NewEnumerator0
    Lifetime=0.25
    Mesh Scale=&Lerp_Vector(A=<1.2 1 1>, B=<1.5 1 1>, Alpha=&ReturnNormalizedExecIndex())
    Mesh Scale Mode=::NewEnumerator4
    Position Mode=::NewEnumerator0
    Sprite Size Mode=::NewEnumerator3
    Uniform Sprite Size=50
    Write Material Random=true
   )
   &SocketLocation(
    SamplePosOffset=true
    Sampled Position Offset=&Lerp_Vector(A=<100 0 0>, B=<160 0 -2>, Alpha=&ReturnNormalizedExecIndex())
    Socket Index=1
    Socket Reader=$System.Socket Reader
    Socket Sampling Mode=::NewEnumerator3
    VelocitySpeedLimit=false
   ) {module=SocketLocation001}
  ]
  particleUpdateScript [
   &ParticleState()
   &ScaleMeshSize(Scale Factor=&VectorFromCurve(VectorCurve=DataInterface() {propertyValues {
       XCurve {
        defaultValue=0.15
        keys [
         {arriveTangent=0.0292, interpMode=RCIM_Cubic, leaveTangent=0.0292, tangentMode=RCTM_User, time=0, value=1}
         {arriveTangent=-2.032, interpMode=RCIM_Cubic, leaveTangent=-0.436, tangentMode=RCTM_Break, time=0.2, value=0.85}
         {arriveTangent=-0.3548, interpMode=RCIM_Cubic, leaveTangent=-1.9197, tangentMode=RCTM_Break, time=0.35, value=0.7}
         {arriveTangent=-0.1352, interpMode=RCIM_Cubic, leaveTangent=-0.1352, tangentMode=RCTM_User, time=1, value=0.25}
        ]
       }
       YCurve {keys [<at 0 1 RCIM_Cubic>, {interpMode=RCIM_Cubic, leaveTangent=1, time=1, value=0.5}]}
       ZCurve {keys [<at 0 1>,<at 1 1>]}
      }}))
   &SocketLocation(
    SamplePosOffset=true
    Sampled Position Offset=&FN_FrontEndVector3Modifier(
     Gameplay=&Lerp_Vector(A=<100 0 0>, B=<160 0 -2>, Alpha=&ReturnNormalizedExecIndex())
     Front End=&Lerp_Vector(A=<100 0 -2>, B=<160 0 -3>, Alpha=&ReturnNormalizedExecIndex())
    )
    Socket Index=1
    Socket Reader=$System.Socket Reader
    Socket Sampling Mode=::NewEnumerator3
    VelocitySpeedLimit=false
   )
   &DynamicMaterialParameters(
    Index 0 Param 1=&FloatFromCurve(FloatCurve=DataInterface() {propertyValues {Curve {defaultValue=1, keys [{tangentMode=RCTM_User, time=0, value=0}, {tangentMode=RCTM_User, time=1, value=1.2}]}}})
    Index 0 Param 3=1
    Index 0 Param 4=&FloatFromCurve(FloatCurve=DataInterface() {propertyValues {Curve {keys [
         {arriveTangent=-13.5777, interpMode=RCIM_Cubic, leaveTangent=-13.5777, tangentMode=RCTM_User, time=0, value=10}
         {arriveTangent=-0.8161, interpMode=RCIM_Cubic, leaveTangent=-0.8161, tangentMode=RCTM_User, time=1, value=2}
        ]}}})
    Module.Index0Float0Write=true
    Module.Index0Float1Write=true
    Module.Index0Float2Write=true
    Module.Index0Float3Write=true
    Module.Index1Float0Write=true
    Module.Index1Float1Write=true
    Module.Index1Float2Write=true
    Module.Index1Float3Write=true
    Write Parameter Index 1=::NewEnumerator1
   )
   &ScaleColor(
    Scale Alpha=&Multiply_Float(
     A=&MakeFloatFromLinearColor(LinearColor=$System.TODColorMultiplier, Channel:ENiagaraFloatFromLinearColorOptions=::NewEnumerator3)
     B=&FloatFromCurve(FloatCurve=DataInterface() {propertyValues {Curve {
         defaultValue=0
         keys [{interpMode=RCIM_Cubic, tangentMode=RCTM_User, time=0, value=1}, {interpMode=RCIM_Cubic, tangentMode=RCTM_User, time=1, value=0}]
        }}})
    )
    Scale RGB=&MakeVectorFromLinearColorRGB(Color=$System.TODColorMultiplier)
    ScaleA=true
    ScaleRGB=true
   )
  ]
  rendererClasses [&NiagaraSpriteRendererProperties, &NiagaraMeshRendererProperties]
  renderers [
   {rendererClass=&NiagaraSpriteRendererProperties}
   {rendererClass=&NiagaraMeshRendererProperties, rendererIndex=1}
  ]
 }
 NE_DarkFlare {
  emitterUpdateScript [
   &EmitterState(
    Inactive Response=::NewEnumerator1
    Life Cycle Mode=::NewEnumerator1
    Loop Behavior=::NewEnumerator1
    Loop Delay=$User.DeployDelay
    Loop Duration=3
    UseLoopDelay=true
   )
   &SpawnBurst_Instantaneous(Use Spawn Probability=false, UseLoopCountLimit=false)
  ]
  particleSpawnScript [
   &InitializeParticle(
    Color=&LerpLinearColors(StartColor=<rgba 0.5423 0.92 1 1>, EndColor=<rgba 0.5423 0.92 1 1>, LerpFactor=$User.Variant)
    Lifetime=0.5
    Mass Mode=::NewEnumerator1
    Mesh Scale Mode=::NewEnumerator3
    Mesh Uniform Scale=0.2
    Position Mode=::NewEnumerator0
    Sprite Rotation Mode=::NewEnumerator2
    Sprite Size=&Vector2DFromFloat(Value=&Multiply_Float(A=750, B=$System.DeployScale))
    Sprite Size Mode=::NewEnumerator4
    Write Material Random=true
   ) {module=InitializeParticle001}
   &CameraOffset(Camera Offset Amount=75)
  ]
  particleUpdateScript [
   &ParticleState()
   &SocketLocation(
    Interpolate Socket Sampling=true
    SamplePosOffset=false
    Socket Reader=$System.Socket Reader
    Socket Sampling Mode=::NewEnumerator3
    VelocitySpeedLimit=false
   )
   &SolveForcesAndVelocity(Clamp Velocity=false, Limit Acceleration=false)
   &ScaleSpriteSize(
    Scale Factor=&Vector2DFromFloat(Value=&FloatFromCurve(
      FloatCurve=DataInterface() {propertyValues {Curve {keys [
          {arriveTangent=0.7585, interpMode=RCIM_Cubic, leaveTangent=0.7585, tangentMode=RCTM_User, time=0, value=0.4}
          {interpMode=RCIM_Cubic, tangentMode=RCTM_User, time=1, value=1}
         ]}}}
      Scale Curve=0.1
     ))
    Scale Sprite Size Mode=::NewEnumerator2
   )
   &SpriteRotationRate(Rotation Rate=&UniformRangedFloat(Minimum=-50, Maximum=50))
   /SetVariables(Particles.DarkGlowSize:Vector2f=&Multiply_Vector2D_ByFloat(Float=3, Vector2D=$Particles.SpriteSize)) {isSetParameters=true}
   /SetVariables(Particles.DargGlowColor:LinearColor=&MakeLinearColorFromVectorAndFloat(
     "Vector (RGB)"=&MakeVectorFromLinearColorRGB(Color=<rgba 0.0452 0.0406 0.05 1>)
     "Float (Alpha)"=&FloatFromCurve(FloatCurve=DataInterface() {propertyValues {Curve {keys [
          {arriveTangent=-0.0975, interpMode=RCIM_Cubic, leaveTangent=-0.0975, tangentMode=RCTM_User, time=0, value=1}
          {interpMode=RCIM_Cubic, leaveTangent=0.002, time=1, value=0}
         ]}}})
    )) {isSetParameters=true}
   /SetVariables(Particles.CentralFlareSize:Vector2f=&Multiply_Vector2D_ByFloat(Float=3, Vector2D=$Particles.SpriteSize)) {isSetParameters=true}
   /SetVariables(Particles.CentralFlareColor:LinearColor=&MakeLinearColorFromVectorAndFloat(
     "Vector (RGB)"=&MakeVectorFromLinearColorRGB(Color=&LerpLinearColors(
       StartColor=&ColorFromCurve(ColorCurve=DataInterface() {propertyValues {
          AlphaCurve {keys [<at 0 1 RCIM_Cubic>,<at 1 0 RCIM_Cubic>]}
          BlueCurve {keys [<at 0 10>, {interpMode=RCIM_Cubic, leaveTangent=1.3364, time=0.6917, value=2}]}
          GreenCurve {keys [<at 0 3.3999>, {interpMode=RCIM_Cubic, leaveTangent=-5.7627, time=0.6917, value=0.5037}]}
          RedCurve {keys [<at 0 0.8711>, {interpMode=RCIM_Cubic, leaveTangent=-13.364, time=0.6917, value=1.1678}]}
         }})
       EndColor=&ColorFromCurve(ColorCurve=DataInterface() {propertyValues {
          AlphaCurve {keys [<at 0 1 RCIM_Cubic>,<at 1 0 RCIM_Cubic>]}
          BlueCurve {keys [<at 0 10>, <at 0.3261 6.2213>, {interpMode=RCIM_Cubic, leaveTangent=1.3364, time=0.6917, value=1.9854}]}
          GreenCurve {keys [<at 0 9.2393>, <at 0.3261 1.6344>, {interpMode=RCIM_Cubic, leaveTangent=-5.7627, time=0.6917, value=0.413}]}
          RedCurve {keys [<at 0 0.8711>, <at 0.3261 1.6923>, {interpMode=RCIM_Cubic, leaveTangent=-13.364, time=0.6917, value=2}]}
         }})
       LerpFactor=$User.Variant
      ))
     "Float (Alpha)"=&FloatFromCurve(
      FloatCurve=DataInterface() {propertyValues {Curve {keys [
          <at 0 0 RCIM_Cubic>
          {arriveTangent=0.3854, interpMode=RCIM_Cubic, leaveTangent=0.3854, tangentMode=RCTM_User, time=0.1, value=1}
          <at 1 0 RCIM_Cubic>
         ]}}}
      Scale Curve=&Lerp_Float(A=0.5, B=&FN_FrontEndFloatModifier(Gameplay=0.5, Front End=0.3), Alpha=$User.Variant)
     )
    )) {isSetParameters=true}
   &ScaleColor(
    Scale Alpha=&Multiply_Float(
     A=&MakeFloatFromLinearColor(LinearColor=$System.TODColorMultiplier, Channel:ENiagaraFloatFromLinearColorOptions=::NewEnumerator3)
     B=&FloatFromCurve(
      FloatCurve=DataInterface() {propertyValues {Curve {keys [
          {arriveTangent=6.4019, interpMode=RCIM_Cubic, leaveTangent=6.4019, tangentMode=RCTM_User, time=0, value=1}
          {interpMode=RCIM_Cubic, leaveTangent=0.002, time=1, value=0}
         ]}}}
      Scale Curve=0.02
     )
    )
    Scale RGB=&Multiply_Vector(
     A=&MakeVectorFromLinearColorRGB(Color=$Particles.CentralFlareColor)
     B=&MakeVectorFromLinearColorRGB(Color=$System.TODColorMultiplier)
    )
    ScaleA=true
    ScaleRGB=true
   )
   &FN_Light_Attributes(
    Color=$Particles.CentralFlareColor
    Create Color Attribute=true
    Create Exponent Attribute=true
    Create Radius Attribute=true
    Exponent=1
    Radius=&Multiply_Float(
     A=&FloatFromCurve(
      FloatCurve=DataInterface() {propertyValues {Curve {keys [
          {arriveTangent=2.3282, interpMode=RCIM_Cubic, leaveTangent=2.3282, tangentMode=RCTM_User, time=0, value=0}
          {interpMode=RCIM_Cubic, tangentMode=RCTM_User, time=1, value=1}
         ]}}}
      Scale Curve=150
     )
     B=$System.DeployScale
    )
    Scale Color=&FloatFromCurve(
     FloatCurve=DataInterface() {propertyValues {Curve {keys [
         {interpMode=RCIM_Cubic, tangentMode=RCTM_User, time=0, value=1}
         {arriveTangent=-3.4524, interpMode=RCIM_Cubic, leaveTangent=-3.4524, tangentMode=RCTM_User, time=1, value=0}
        ]}}}
     Scale Curve=7
    )
   )
   &DynamicMaterialParameters(
    Index 1 Param 1=0.3
    Module.Index1Float0Write=true
    Module.Index1Float1Write=true
    Module.Index1Float2Write=true
    Module.Index1Float3Write=true
    Module.Index3Float0Write=true
    Module.Index3Float1Write=true
    Module.Index3Float2Write=true
    Module.Index3Float3Write=true
    Write Parameter Index 0=::NewEnumerator0
    Write Parameter Index 1=::NewEnumerator1
    Write Parameter Index 3=::NewEnumerator1
   )
   &FN_Sprite_Attributes(
    Create Sprite Rotation Attribute=true
    Create SpriteSize Attribute=true
    SpriteSize=&Multiply_Vector2D(A=$Particles.CentralFlareSize, B=<2 0.15>)
    Write to Module Namespace=true
   )
  ]
  rendererClasses [&NiagaraSpriteRendererProperties, &NiagaraLightRendererProperties]
  renderers [
   {rendererClass=&NiagaraSpriteRendererProperties}
   {rendererClass=&NiagaraLightRendererProperties, rendererIndex=1}
   {rendererClass=&NiagaraSpriteRendererProperties, rendererIndex=2}
   {rendererClass=&NiagaraSpriteRendererProperties, rendererIndex=3}
  ]
 }
 NE_SoundRings {
  emitterUpdateScript [
   &EmitterState(
    Inactive Response=::NewEnumerator1
    Life Cycle Mode=::NewEnumerator1
    Loop Behavior=::NewEnumerator1
    Loop Delay=$User.DeployDelay
    Loop Duration=3
    UseLoopDelay=true
   )
   &SpawnBurst_Instantaneous(
    Spawn Count=3
    Spawn Probability=&FN_FrontEndFloatModifier(Gameplay=1, Front End=0)
    Use Spawn Probability=true
    UseLoopCountLimit=false
   )
  ]
  particleSpawnScript [
   &InitializeParticle(
    Color=<rgba 1 1 1 2>
    Lifetime=0.35
    Position Offset=$System.SystemOffset
    Sprite Size=&Multiply_Vector2D_ByFloat(
     Float=&Multiply_Float(
      A=&Lerp_Float(A=150, B=200, Alpha=&ReturnNormalizedExecIndex())
      B=&FN_FrontEndFloatModifier(Gameplay=3, Front End=2)
     )
     Vector2D=<1 1>
    )
    Sprite Size Mode=::NewEnumerator4
    UsePositionOffset=true
    Write Material Random=true
   )
   /SetVariables(Particles.BaseNormalisedIndex:NiagaraFloat=&ReturnNormalizedExecIndex()) {isSetParameters=true, module=SetVariables_#8}
   /SetVariables(Particles.Age:NiagaraFloat=&Multiply_Float(A=&OneMinusFloat(Float=&ReturnNormalizedExecIndex()), B=-0.1)) {isSetParameters=true, module=SetVariables_#9}
   &CameraOffset()
   &SocketLocation(
    Interpolate Socket Sampling=true
    Orientation Sampling=::NewEnumerator1
    SamplePosOffset=false
    Socket Reader=$System.Socket Reader
    Socket Sampling Mode=::NewEnumerator3
    Velocity Sampling:ENiagara_AttributeSamplingApplyOutput=::NewEnumerator0
    VelocitySpeedLimit=false
   ) {module=SocketLocation001}
  ]
  particleUpdateScript [
   &ParticleState()
   &SocketLocation(
    Interpolate Socket Sampling=true
    SamplePosOffset=false
    Sampled Orientation Coordinate Space:ENiagaraCoordinateSpace=::Local
    Socket Reader=$System.Socket Reader
    Socket Sampling Mode=::NewEnumerator3
    VelocitySpeedLimit=false
   )
   &ScaleSpriteSize(
    Scale Sprite Size Mode=::NewEnumerator0
    Uniform Scale Factor=&Multiply_Float(
     A=&FloatFromCurve(FloatCurve=DataInterface() {propertyValues {Curve {keys [
          {arriveTangent=1.8909, interpMode=RCIM_Cubic, leaveTangent=1.8909, tangentMode=RCTM_User, time=0, value=0}
          {interpMode=RCIM_Cubic, tangentMode=RCTM_User, time=1, value=1}
         ]}}})
     B=1
    )
   )
   &DynamicMaterialParameters(
    Index 0 Param 1=0.2
    Index 0 Param 2=1
    Index 0 Param 4=&FloatFromCurve(FloatCurve=DataInterface() {propertyValues {Curve {defaultValue=1, keys [{tangentMode=RCTM_User, time=0, value=0}, {tangentMode=RCTM_User, time=1, value=1}]}}})
    Module.Index0Float0Write=true
    Module.Index0Float1Write=true
    Module.Index0Float2Write=true
    Module.Index0Float3Write=true
   )
   &ScaleColor(
    Scale Alpha=&Multiply_Float(
     A=&MakeFloatFromLinearColor(LinearColor=$System.TODColorMultiplier, Channel:ENiagaraFloatFromLinearColorOptions=::NewEnumerator3)
     B=&FloatFromCurve(
      FloatCurve=DataInterface() {propertyValues {Curve {keys [
          {arriveTangent=4.2862, interpMode=RCIM_Cubic, leaveTangent=4.2862, tangentMode=RCTM_User, time=0, value=0}
          {interpMode=RCIM_Cubic, tangentMode=RCTM_User, time=0.1, value=1}
          {arriveTangent=-0.5134, interpMode=RCIM_Cubic, leaveTangent=-0.5134, tangentMode=RCTM_User, time=1, value=0}
         ]}}}
      Scale Curve=0.2
     )
    )
    Scale RGB=&MakeVectorFromLinearColorRGB(Color=&MultiplyLinearColor(
      Linear Color=&Multiply_LinearColorByFloat(
       Linear Color=&LerpLinearColors(
        StartColor=&ColorFromCurve(
         ColorCurve=DataInterface() {propertyValues {
           AlphaCurve {keys [<at 0 1 RCIM_Cubic>]}
           BlueCurve {keys [<at 0 10 RCIM_Cubic>,<at 1 2 RCIM_Cubic>]}
           GreenCurve {keys [<at 0 0 RCIM_Cubic>,<at 1 0.6319 RCIM_Cubic>]}
           RedCurve {keys [<at 0 1.939 RCIM_Cubic>,<at 1 1.5249 RCIM_Cubic>]}
          }}
         CurveIndex=$Particles.BaseNormalisedIndex
        )
        EndColor=&ColorFromCurve(
         ColorCurve=DataInterface() {propertyValues {
           AlphaCurve {keys [<at 0 1 RCIM_Cubic>]}
           BlueCurve {keys [<at 0.2836 10 RCIM_Cubic>,<at 1 2 RCIM_Cubic>]}
           GreenCurve {keys [<at 0.2836 1.3623 RCIM_Cubic>,<at 1 0.4232 RCIM_Cubic>]}
           RedCurve {keys [<at 0.2836 7.1208 RCIM_Cubic>,<at 1 2 RCIM_Cubic>]}
          }}
         CurveIndex=$Particles.BaseNormalisedIndex
        )
        LerpFactor=$User.Variant
       )
       Float=&Lerp_Float(
        A=0
        B=2
        Alpha=&FloatFromCurve(FloatCurve=DataInterface() {propertyValues {Curve {keys [
             {arriveTangent=-1.6896, interpMode=RCIM_Cubic, leaveTangent=-1.6896, tangentMode=RCTM_User, time=0, value=1}
             {arriveTangent=0.0045, interpMode=RCIM_Cubic, leaveTangent=0.0045, tangentMode=RCTM_User, time=1, value=0}
            ]}}})
       )
      )
      Scale Factor=&MakeVector4_FromLinearColor(LinearColor=$System.TODColorMultiplier)
     ))
    ScaleA=true
    ScaleRGB=true
   )
   &SpriteFacingAndAlignment(Sprite Facing=<0 0 1>, WriteAlignment=false, WriteFacing=true)
   &FN_Sprite_Attributes(
    Color=&Multiply_LinearColorByFloat(
     Linear Color=&LerpLinearColors(
      StartColor=&ColorFromCurve(
       ColorCurve=DataInterface() {propertyValues {
         AlphaCurve {keys [<at 0 1 RCIM_Cubic>]}
         BlueCurve {keys [<at 0 10 RCIM_Cubic>,<at 0.2762 3.9582>,<at 0.6961 2 RCIM_Cubic>]}
         GreenCurve {keys [<at 0 4.3518 RCIM_Cubic>,<at 0.2762 2.2373>,<at 0.6961 1.5365 RCIM_Cubic>]}
         RedCurve {keys [<at 0 1.5277 RCIM_Cubic>,<at 0.2762 0.5164>,<at 0.6961 0.6095 RCIM_Cubic>]}
        }}
       CurveIndex=$Particles.BaseNormalisedIndex
      )
      EndColor=&ColorFromCurve(
       ColorCurve=DataInterface() {propertyValues {
         AlphaCurve {keys [<at 0 1 RCIM_Cubic>]}
         BlueCurve {keys [<at 0 10 RCIM_Cubic>,<at 0.2762 3.9582>,<at 0.6961 2 RCIM_Cubic>]}
         GreenCurve {keys [<at 0 10 RCIM_Cubic>,<at 0.2762 3.3846>,<at 0.6961 1.5365 RCIM_Cubic>]}
         RedCurve {keys [<at 0 1.5277 RCIM_Cubic>,<at 0.2762 0.5164>,<at 0.6961 0.6095 RCIM_Cubic>]}
        }}
       CurveIndex=$Particles.BaseNormalisedIndex
      )
      LerpFactor=$User.Variant
     )
     Float=&Lerp_Float(
      A=0
      B=2
      Alpha=&FloatFromCurve(FloatCurve=DataInterface() {propertyValues {Curve {keys [
           {arriveTangent=-1.6896, interpMode=RCIM_Cubic, leaveTangent=-1.6896, tangentMode=RCTM_User, time=0, value=1}
           {arriveTangent=0.0045, interpMode=RCIM_Cubic, leaveTangent=0.0045, tangentMode=RCTM_User, time=1, value=0}
          ]}}})
     )
    )
    Create Color Attribute=true
    Create SpriteSize Attribute=true
    SpriteSize=&Vector2DFromFloat(Value=&Multiply_Float(
      A=&Multiply_Float(A=&Lerp_Float(A=100, B=10, Alpha=$Particles.BaseNormalisedIndex), B=$System.DeployScale)
      B=&FloatFromCurve(FloatCurve=DataInterface() {propertyValues {Curve {keys [
           {arriveTangent=2.3035, interpMode=RCIM_Cubic, leaveTangent=2.3035, tangentMode=RCTM_User, time=0, value=0}
           {interpMode=RCIM_Cubic, tangentMode=RCTM_User, time=1, value=1}
          ]}}})
     ))
    Write to Module Namespace=true
   )
  ]
  rendererClasses [&NiagaraSpriteRendererProperties]
  renderers [
   {rendererClass=&NiagaraSpriteRendererProperties}
   {rendererClass=&NiagaraSpriteRendererProperties, rendererIndex=1}
  ]
 }
 NE_SoundRingsFE {
  emitterUpdateScript [
   &EmitterState(
    Inactive Response=::NewEnumerator1
    Life Cycle Mode=::NewEnumerator1
    Loop Behavior=::NewEnumerator1
    Loop Delay=$User.DeployDelay
    Loop Duration=3
    UseLoopDelay=true
   )
   &SpawnBurst_Instantaneous(
    Spawn Count=3
    Spawn Probability=&FN_FrontEndFloatModifier(Gameplay=0, Front End=1)
    Use Spawn Probability=true
    UseLoopCountLimit=false
   )
  ]
  particleSpawnScript [
   &InitializeParticle(
    Color=<rgba 1 1 1 2>
    Lifetime=0.35
    Position Offset=$System.SystemOffset
    Sprite Size=&Multiply_Vector2D_ByFloat(
     Float=&Multiply_Float(
      A=&Lerp_Float(A=150, B=200, Alpha=&ReturnNormalizedExecIndex())
      B=&FN_FrontEndFloatModifier(Gameplay=3, Front End=2)
     )
     Vector2D=<1 1>
    )
    Sprite Size Mode=::NewEnumerator4
    UsePositionOffset=true
    Write Material Random=true
   )
   /SetVariables(Particles.BaseNormalisedIndex:NiagaraFloat=&ReturnNormalizedExecIndex()) {isSetParameters=true, module=SetVariables_#8}
   /SetVariables(Particles.Age:NiagaraFloat=&Multiply_Float(A=&OneMinusFloat(Float=&ReturnNormalizedExecIndex()), B=-0.1)) {isSetParameters=true, module=SetVariables_#9}
   &CameraOffset()
   &SocketLocation(
    Interpolate Socket Sampling=true
    Orientation Sampling=::NewEnumerator1
    SamplePosOffset=false
    Socket Reader=$System.Socket Reader
    Socket Sampling Mode=::NewEnumerator3
    Velocity Sampling:ENiagara_AttributeSamplingApplyOutput=::NewEnumerator0
    VelocitySpeedLimit=false
   ) {module=SocketLocation001}
  ]
  particleUpdateScript [
   &ParticleState()
   &SocketLocation(
    Interpolate Socket Sampling=true
    SamplePosOffset=false
    Sampled Orientation Coordinate Space:ENiagaraCoordinateSpace=::Local
    Socket Reader=$System.Socket Reader
    Socket Sampling Mode=::NewEnumerator3
    VelocitySpeedLimit=false
   )
   &ScaleSpriteSize(
    Scale Sprite Size Mode=::NewEnumerator0
    Uniform Scale Factor=&Multiply_Float(
     A=&FloatFromCurve(FloatCurve=DataInterface() {propertyValues {Curve {keys [
          {arriveTangent=1.8909, interpMode=RCIM_Cubic, leaveTangent=1.8909, tangentMode=RCTM_User, time=0, value=0}
          {interpMode=RCIM_Cubic, tangentMode=RCTM_User, time=1, value=1}
         ]}}})
     B=1
    )
   )
   &DynamicMaterialParameters(
    Index 0 Param 1=0.2
    Index 0 Param 2=1
    Index 0 Param 4=&FloatFromCurve(FloatCurve=DataInterface() {propertyValues {Curve {defaultValue=1, keys [{tangentMode=RCTM_User, time=0, value=0}, {tangentMode=RCTM_User, time=1, value=1}]}}})
    Module.Index0Float0Write=true
    Module.Index0Float1Write=true
    Module.Index0Float2Write=true
    Module.Index0Float3Write=true
   )
   &ScaleColor(
    Scale Alpha=&Multiply_Float(
     A=&MakeFloatFromLinearColor(LinearColor=$System.TODColorMultiplier, Channel:ENiagaraFloatFromLinearColorOptions=::NewEnumerator3)
     B=&FloatFromCurve(
      FloatCurve=DataInterface() {propertyValues {Curve {keys [
          {arriveTangent=4.2862, interpMode=RCIM_Cubic, leaveTangent=4.2862, tangentMode=RCTM_User, time=0, value=0}
          {interpMode=RCIM_Cubic, tangentMode=RCTM_User, time=0.1, value=1}
          {arriveTangent=-0.5134, interpMode=RCIM_Cubic, leaveTangent=-0.5134, tangentMode=RCTM_User, time=1, value=0}
         ]}}}
      Scale Curve=0.2
     )
    )
    Scale RGB=&MakeVectorFromLinearColorRGB(Color=&MultiplyLinearColor(
      Linear Color=&Multiply_LinearColorByFloat(
       Linear Color=&LerpLinearColors(
        StartColor=&ColorFromCurve(
         ColorCurve=DataInterface() {propertyValues {
           AlphaCurve {keys [<at 0 1 RCIM_Cubic>]}
           BlueCurve {keys [<at 0 10 RCIM_Cubic>,<at 1 2 RCIM_Cubic>]}
           GreenCurve {keys [<at 0 0 RCIM_Cubic>,<at 1 0.6319 RCIM_Cubic>]}
           RedCurve {keys [<at 0 1.939 RCIM_Cubic>,<at 1 1.5249 RCIM_Cubic>]}
          }}
         CurveIndex=$Particles.BaseNormalisedIndex
        )
        EndColor=&ColorFromCurve(
         ColorCurve=DataInterface() {propertyValues {
           AlphaCurve {keys [<at 0 1 RCIM_Cubic>]}
           BlueCurve {keys [<at 0.1796 10 RCIM_Cubic>,<at 0.9896 2 RCIM_Cubic>]}
           GreenCurve {keys [<at 0.1796 1.3623 RCIM_Cubic>,<at 0.9896 0.4232 RCIM_Cubic>]}
           RedCurve {keys [<at 0.1796 7.1208 RCIM_Cubic>,<at 0.9896 2 RCIM_Cubic>]}
          }}
         CurveIndex=$Particles.BaseNormalisedIndex
        )
        LerpFactor=$User.Variant
       )
       Float=&Lerp_Float(
        A=0
        B=2
        Alpha=&FloatFromCurve(FloatCurve=DataInterface() {propertyValues {Curve {keys [
             {arriveTangent=-1.6896, interpMode=RCIM_Cubic, leaveTangent=-1.6896, tangentMode=RCTM_User, time=0, value=1}
             {arriveTangent=0.0045, interpMode=RCIM_Cubic, leaveTangent=0.0045, tangentMode=RCTM_User, time=1, value=0}
            ]}}})
       )
      )
      Scale Factor=&MakeVector4_FromLinearColor(LinearColor=$System.TODColorMultiplier)
     ))
    ScaleA=true
    ScaleRGB=true
   )
   &SpriteFacingAndAlignment(Sprite Facing=<0 0 1>, WriteAlignment=false, WriteFacing=true)
   &FN_Sprite_Attributes(
    Color=&Multiply_LinearColorByFloat(
     Linear Color=&LerpLinearColors(
      StartColor=&ColorFromCurve(
       ColorCurve=DataInterface() {propertyValues {
         AlphaCurve {keys [<at 0 1 RCIM_Cubic>]}
         BlueCurve {keys [<at 0 10 RCIM_Cubic>,<at 0.2762 3.9582>,<at 0.6961 2 RCIM_Cubic>]}
         GreenCurve {keys [<at 0 4.3518 RCIM_Cubic>,<at 0.2762 2.2373>,<at 0.6961 1.5365 RCIM_Cubic>]}
         RedCurve {keys [<at 0 1.5277 RCIM_Cubic>,<at 0.2762 0.5164>,<at 0.6961 0.6095 RCIM_Cubic>]}
        }}
       CurveIndex=$Particles.BaseNormalisedIndex
      )
      EndColor=&ColorFromCurve(
       ColorCurve=DataInterface() {propertyValues {
         AlphaCurve {keys [<at 0 1 RCIM_Cubic>]}
         BlueCurve {keys [<at 0 10 RCIM_Cubic>,<at 0.2762 3.9582>,<at 0.6961 2 RCIM_Cubic>]}
         GreenCurve {keys [<at 0 10 RCIM_Cubic>,<at 0.2762 3.3846>,<at 0.6961 1.5365 RCIM_Cubic>]}
         RedCurve {keys [<at 0 1.5277 RCIM_Cubic>,<at 0.2762 0.5164>,<at 0.6961 0.6095 RCIM_Cubic>]}
        }}
       CurveIndex=$Particles.BaseNormalisedIndex
      )
      LerpFactor=$User.Variant
     )
     Float=&Lerp_Float(
      A=0
      B=2
      Alpha=&FloatFromCurve(FloatCurve=DataInterface() {propertyValues {Curve {keys [
           {arriveTangent=-1.6896, interpMode=RCIM_Cubic, leaveTangent=-1.6896, tangentMode=RCTM_User, time=0, value=1}
           {arriveTangent=0.0045, interpMode=RCIM_Cubic, leaveTangent=0.0045, tangentMode=RCTM_User, time=1, value=0}
          ]}}})
     )
    )
    Create Color Attribute=true
    Create SpriteSize Attribute=true
    SpriteSize=&Vector2DFromFloat(Value=&Multiply_Float(
      A=&Multiply_Float(A=&Lerp_Float(A=100, B=10, Alpha=$Particles.BaseNormalisedIndex), B=$System.DeployScale)
      B=&FloatFromCurve(FloatCurve=DataInterface() {propertyValues {Curve {keys [
           {arriveTangent=2.3035, interpMode=RCIM_Cubic, leaveTangent=2.3035, tangentMode=RCTM_User, time=0, value=0}
           {interpMode=RCIM_Cubic, tangentMode=RCTM_User, time=1, value=1}
          ]}}})
     ))
    Write to Module Namespace=true
   )
  ]
  rendererClasses [&NiagaraSpriteRendererProperties]
  renderers [
   {rendererClass=&NiagaraSpriteRendererProperties}
   {rendererClass=&NiagaraSpriteRendererProperties, rendererIndex=1}
  ]
 }
 RightWing {
  emitterSpawnScript [/SetVariables(
    Emitter.CenterColor:LinearColor=&LerpLinearColors(StartColor=<rgba 1.0887 0.3246 2 1>, EndColor=<rgba 1.7272 0.3631 2 1>, LerpFactor=$User.Variant)
    Emitter.EdgeColor:LinearColor=&LerpLinearColors(StartColor=<rgba 2 0.713 1.8132 1>, EndColor=<rgba 1 0 3 1>, LerpFactor=$User.Variant)
    Emitter.TipColor:LinearColor=&LerpLinearColors(StartColor=<rgba 0.2218 0.6687 1.5 1>, EndColor=<rgba 0.2218 1.5 1.5 1>, LerpFactor=$User.Variant)
   ) {isSetParameters=true, module=SetVariables_#7}]
  emitterUpdateScript [
   &EmitterState(
    Life Cycle Mode=::NewEnumerator1
    Loop Behavior=::NewEnumerator1
    Loop Delay=&FN_FrontEndFloatModifier(Gameplay=2.1, Front End=1.83)
    Loop Duration=1
    UseLoopDelay=true
   )
   &SpawnBurst_Instantaneous(Spawn Count=3, Use Spawn Probability=false, UseLoopCountLimit=false)
  ]
  particleSpawnScript [
   &InitializeParticle(
    Color Mode=::NewEnumerator0
    Lifetime=0.25
    Mesh Scale=&Lerp_Vector(A=<1.2 1 1>, B=<1.5 1 1>, Alpha=&ReturnNormalizedExecIndex())
    Mesh Scale Mode=::NewEnumerator4
    Position Mode=::NewEnumerator0
    Sprite Size Mode=::NewEnumerator3
    Uniform Sprite Size=50
    Write Material Random=true
   )
   &SocketLocation(
    SamplePosOffset=true
    Sampled Position Offset=&Lerp_Vector(A=<100 0 0>, B=<160 0 2>, Alpha=&ReturnNormalizedExecIndex())
    Socket Index=2
    Socket Reader=$System.Socket Reader
    Socket Sampling Mode=::NewEnumerator3
    VelocitySpeedLimit=false
   ) {module=SocketLocation001}
  ]
  particleUpdateScript [
   &ParticleState()
   &ScaleMeshSize(Scale Factor=&VectorFromCurve(VectorCurve=DataInterface() {propertyValues {
       XCurve {
        defaultValue=0.15
        keys [
         {arriveTangent=0.0292, interpMode=RCIM_Cubic, leaveTangent=0.0292, tangentMode=RCTM_User, time=0, value=1}
         {arriveTangent=-2.032, interpMode=RCIM_Cubic, leaveTangent=-0.436, tangentMode=RCTM_Break, time=0.2, value=0.85}
         {arriveTangent=-0.3548, interpMode=RCIM_Cubic, leaveTangent=-1.9197, tangentMode=RCTM_Break, time=0.35, value=0.7}
         {arriveTangent=-0.1352, interpMode=RCIM_Cubic, leaveTangent=-0.1352, tangentMode=RCTM_User, time=1, value=0.25}
        ]
       }
       YCurve {keys [<at 0 1 RCIM_Cubic>, {interpMode=RCIM_Cubic, leaveTangent=1, time=1, value=0.5}]}
       ZCurve {keys [<at 0 1>,<at 1 1>]}
      }}))
   &SocketLocation(
    SamplePosOffset=true
    Sampled Position Offset=&FN_FrontEndVector3Modifier(
     Gameplay=&Lerp_Vector(A=<100 0 0>, B=<160 0 2>, Alpha=&ReturnNormalizedExecIndex())
     Front End=&Lerp_Vector(A=<100 0 2>, B=<160 0 3>, Alpha=&ReturnNormalizedExecIndex())
    )
    Socket Index=2
    Socket Reader=$System.Socket Reader
    Socket Sampling Mode=::NewEnumerator3
    VelocitySpeedLimit=false
   )
   &DynamicMaterialParameters(
    Index 0 Param 1=&FloatFromCurve(FloatCurve=DataInterface() {propertyValues {Curve {defaultValue=1, keys [{tangentMode=RCTM_User, time=0, value=0}, {tangentMode=RCTM_User, time=1, value=1.2}]}}})
    Index 0 Param 3=1
    Index 0 Param 4=&FloatFromCurve(FloatCurve=DataInterface() {propertyValues {Curve {keys [
         {arriveTangent=-13.5777, interpMode=RCIM_Cubic, leaveTangent=-13.5777, tangentMode=RCTM_User, time=0, value=10}
         {arriveTangent=-0.8161, interpMode=RCIM_Cubic, leaveTangent=-0.8161, tangentMode=RCTM_User, time=1, value=2}
        ]}}})
    Module.Index0Float0Write=true
    Module.Index0Float1Write=true
    Module.Index0Float2Write=true
    Module.Index0Float3Write=true
    Module.Index1Float0Write=true
    Module.Index1Float1Write=true
    Module.Index1Float2Write=true
    Module.Index1Float3Write=true
    Write Parameter Index 1=::NewEnumerator1
   )
   &ScaleColor(
    Scale Alpha=&Multiply_Float(
     A=&MakeFloatFromLinearColor(LinearColor=$System.TODColorMultiplier, Channel:ENiagaraFloatFromLinearColorOptions=::NewEnumerator3)
     B=&FloatFromCurve(FloatCurve=DataInterface() {propertyValues {Curve {
         defaultValue=0
         keys [{interpMode=RCIM_Cubic, tangentMode=RCTM_User, time=0, value=1}, {interpMode=RCIM_Cubic, tangentMode=RCTM_User, time=1, value=0}]
        }}})
    )
    Scale RGB=&MakeVectorFromLinearColorRGB(Color=$System.TODColorMultiplier)
    ScaleA=true
    ScaleRGB=true
   )
  ]
  rendererClasses [&NiagaraSpriteRendererProperties, &NiagaraMeshRendererProperties]
  renderers [
   {rendererClass=&NiagaraSpriteRendererProperties}
   {rendererClass=&NiagaraMeshRendererProperties, rendererIndex=1}
  ]
 }
}
summary {
 systemName=NS_PolarityWinnGlider_Deploy
 emitters [
  {emitterName=NE_DarkFlare, rendererClasses [&NiagaraSpriteRendererProperties, &NiagaraLightRendererProperties]}
  {emitterName=NE_SoundRings, rendererClasses [&NiagaraSpriteRendererProperties]}
  {emitterName=Deploy_Trail, rendererClasses [&NiagaraRibbonRendererProperties]}
  {emitterName=RightWing, rendererClasses [&NiagaraSpriteRendererProperties, &NiagaraMeshRendererProperties]}
  {emitterName=LeftWing, rendererClasses [&NiagaraSpriteRendererProperties, &NiagaraMeshRendererProperties]}
  {emitterName=NE_SoundRingsFE, rendererClasses [&NiagaraSpriteRendererProperties]}
 ]
 userVariables [
  User.bIsFrontEnd {type=NiagaraBool}
  User.bIsFrontEndPreview {type=NiagaraBool}
  User.DeployDelay {type=NiagaraFloat, defaultValue=0.8, description=""}
  User.RibbonIDOffset {type=NiagaraInt32}
  User.Variant {type=NiagaraFloat, defaultValue=1, description=""}
 ]
}
